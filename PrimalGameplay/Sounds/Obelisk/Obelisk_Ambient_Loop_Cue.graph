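# Looping ambient Sound Cue for an obelisk in PrimalGameplay.
# A looping wave player feeds a modulator that randomizes pitch, with the lower bound dropped below default.
# The cue overrides attenuation with a short-range logarithmic falloff from the source and a non-spatialized radius close in.
# It routes through the Fort_Quests sound class.
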SoundCue /PrimalGameplay/Sounds/Obelisk/Obelisk_Ambient_Loop_Cue
root=Modulator_0
nodes [
 {id=Modulator_0, children [WavePlayer_0], properties {pitchMin=0.8}}
 {id=WavePlayer_0, properties {bLooping=true, soundWaveAssetPtr=&/Obelisk_Ambient_Loop_01}}
]
properties {
 bOverrideAttenuation=true
 maxDistance=1000
 soundClassObject=&Fort_Quests
 volumeMultiplier=1.008
 attenuationOverrides {distanceAlgorithm=Logarithmic, falloffDistance=1000, nonSpatializedRadiusStart=200, attenuationShapeExtents {x=0}}
}
