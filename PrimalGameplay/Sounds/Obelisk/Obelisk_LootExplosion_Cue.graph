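# Sound cue for the Obelisk loot explosion effect in the Primal gameplay content.
# A single wave player plays the Obelisk_LootExplosion_01 wave with a boosted volume, routed to the Fort_Quests sound class.
# It uses the jump boot attenuation asset and overrides concurrency to allow only one instance at a time.
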SoundCue /PrimalGameplay/Sounds/Obelisk/Obelisk_LootExplosion_Cue
root=WavePlayer_3
nodes [{id=WavePlayer_3, properties {soundWaveAssetPtr=&/Obelisk_LootExplosion_01}}]
properties {
 attenuationSettings=/Game/Sounds/Attenuation/Fort_JumpBoot_Att
 bOverrideConcurrency=true
 maxDistance=3584
 soundClassObject=&Fort_Quests
 volumeMultiplier=2.016
 attenuationOverrides {distanceAlgorithm=Logarithmic, falloffDistance=500, attenuationShapeExtents {x=0}}
 concurrencyOverrides {maxCount=1}
}
