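# Sound cue for an obelisk loot trigger, layering a main trigger wave with three delayed accent sounds through a mixer.
# The trigger plays at once, followed by a rune ambient loop, a pitched-down jump boots sound at half mix volume, and a rising-pitch Zero Point crystal loop.
# Envelopers fade the delayed layers in and taper both loops out.
# The cue uses jump boot attenuation with overrides, the quests sound class and single-instance concurrency.
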SoundCue /PrimalGameplay/Sounds/Obelisk/Obelisk_LootTrigger_Cue
root=Mixer_0
nodes [
 {id=Delay_0, children [Enveloper_0], properties {delayMax=1.2, delayMin=1.2}}
 {id=Delay_1, children [Enveloper_1], properties {delayMax=0.7, delayMin=0.7}}
 {id=Delay_2, children [Enveloper_2], properties {delayMax=0.9, delayMin=0.9}}
 {
  id=Enveloper_0
  children [WavePlayer_0]
  properties {
   pitchCurve {editorCurveData {keys [<at -0.0027 0.3471>,<at 3.5647 1.9853>]}}
   volumeCurve {editorCurveData {keys [<at 0.0072 0.0138>,<at 0.4968 1>,<at 1.9636 1.0223>,<at 3.3885 -0.0699>]}}
  }
 }
 {
  id=Enveloper_1
  children [WavePlayer_1]
  properties {
   pitchCurve {editorCurveData {keys [<at 0.0256 1.0025>]}}
   volumeCurve {editorCurveData {keys [<at 0.0072 0.0138>,<at 0.4968 1>,<at 2.8998 1>,<at 3.6 -0.2444>]}}
  }
 }
 {
  id=Enveloper_2
  children [Modulator_0]
  properties {pitchCurve {editorCurveData {keys [{time=0}]}}, volumeCurve {editorCurveData {keys [<at 0.0537 0.0204>,<at 1.0578 0.9994>]}}}
 }
 {id=Mixer_0, children [WavePlayer_3,Delay_0,Delay_1,Delay_2], properties {inputVolume [1,1,1,0.5]}}
 {id=Modulator_0, children [WavePlayer_2], properties {pitchMax=0.7, pitchMin=0.7, volumeMax=1, volumeMin=1}}
 {
  id=WavePlayer_0
  properties {soundWaveAssetPtr=/FNGameplayCues/Sounds/Fort_GamePlay_Sounds/Player/ZeroPointCrystals/ZeroPointCrystal_Expire_Loop_01}
 }
 {id=WavePlayer_1, properties {soundWaveAssetPtr=/PrimalGameplay/Sounds/Rune/Rune_Ambient_Loop_01}}
 {id=WavePlayer_2, properties {soundWaveAssetPtr=/PrimalGameplay/Sounds/JumpBoots/JumpBoots_Jump_3_01}}
 {id=WavePlayer_3, properties {soundWaveAssetPtr=&/Obelisk_LootTrigger_01}}
]
properties {
 attenuationSettings=/Game/Sounds/Attenuation/Fort_JumpBoot_Att
 bOverrideConcurrency=true
 maxDistance=3584
 soundClassObject=&Fort_Quests
 volumeMultiplier=1.68
 attenuationOverrides {distanceAlgorithm=Logarithmic, falloffDistance=500, attenuationShapeExtents {x=0}}
 concurrencyOverrides {maxCount=1}
}
