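# Sound cue for an obelisk rune stone's in-range sound, classed under quest audio.
# A single rune stone wave plays through a modulator whose volume range is pinned, so it adds no volume variation, and the cue applies a boosted volume multiplier.
# It uses a shared obelisk in-range attenuation asset and overrides concurrency to cap simultaneous instances.
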SoundCue /PrimalGameplay/Sounds/Obelisk/Obelisk_RuneStone_In_Range_Cue
root=Modulator_0
nodes [
 {id=Modulator_0, children [WavePlayer_4], properties {volumeMax=1, volumeMin=1}}
 {id=WavePlayer_4, properties {soundWaveAssetPtr=&/Obelisk_RuneStone_In_Range_01}}
]
properties {
 attenuationSettings=&/Fort_Obelisk_In_Range_Att
 bOverrideConcurrency=true
 maxDistance=6500
 soundClassObject=&Fort_Quests
 volumeMultiplier=2.464
 attenuationOverrides {distanceAlgorithm=Logarithmic, falloffDistance=500, attenuationShapeExtents {x=0}}
 concurrencyOverrides {maxCount=3}
}
